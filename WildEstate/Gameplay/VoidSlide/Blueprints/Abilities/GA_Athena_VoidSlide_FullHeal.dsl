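; Void Slide gameplay ability that restores the owning player's health and shield to maximum in a single activation.
; On activation it caches the avatar pawn and allows healing if a scalable-float setting permits it in the storm or the pawn is inside the safe zone.
; Valid healing applies two effect specs with max health and max shield as SetByCaller magnitudes, plays a healing gameplay cue, then the ability ends.
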(event Ability|EventFailedToActivatePassiveAbility
  (params (ActorInfo :type Gameplay%20Ability%20Actor%20Info%20Structure))
  ;[1 0 0.7498] Required for Passive GAs
  (|CallParentFunction ActorInfo))

(event Ability|EventActivateAbility
  ;[0 0.5547 0.0738] ActivateAbility
  ; Make sure healing is possible, otherwise set the intial delay and begin healing periods
  (|CallParentFunction)
  (bind _outfortplayerpawn (CallFunction|GetAvatarasValidFortPlayerPawn)
    (:ValidFortPlayerPawn
      (bind _output_get (Variables|Default|SetOwningPawn _outfortplayerpawn))
      (Utilities|IsValid _output_get
        (:Is%20Valid
          (bind _ishealingvalid (CallFunction|CheckforValidHealing))
          (if _ishealingvalid
            (CallFunction|HealOwningPlayer)
            ;;[0.5 0.5 0.5] TODO: Visual based on if healing actually occurs? Add bool to HealOwningPlayer function
            (Ability|ExecuteGameplayCueOnOwner (Variables|Default|GetGCNHealing) (Class|AthenaFunctionLibrary|MakeSimpleFortEffectContext (Variables|Default|GetOwningPawn) (Variables|Default|GetOwningPawn)))
            (else
              _))
          (Ability|EndAbility))
        (:Is%20Not%20Valid
          _)))
    (:AvatarCastFailed
      _)))

(fn CheckForValidHealing
  ;[0.3355 0.5518 0.987] Check if healing in the storm is allowed, then if the player is in the storm determine if healing is allowed/not
  (if (ScalableFloat|GetValueasBool (Variables|Default|GetHF_HealingEnabledInStorm))
    ;[1 0.4539 0.904] RETURN is healing valid?
    (return true)
    (else
      (bind _returnvalue (SafeZone|IsLocationinSafeZone (Transformation|GetActorLocation (Variables|Default|GetOwningPawn))))
      (return _returnvalue))))

(fn HealOwningPlayer
  (bind _owningpawn (Variables|Default|GetOwningPawn)
    (:then
      ;[0 0.1037 1] Restore maximum Health
      (bind _returnvalue (Ability|GameplayEffect|AssignTagSetbyCallerMagnitude (Ability|MakeOutgoingGameplayEffectSpec (Variables|Balance|GetGE_RestoreHealth)) "(TagName=\"SetByCaller.Healing\")" (FortPawn|GetMaxHealth _owningpawn)))
      (Ability|ApplyGameplayEffectSpecToOwner _returnvalue)
      ;[0 0.1037 1] Restore maximum Shield
      (bind _returnvalue_2 (Ability|GameplayEffect|AssignTagSetbyCallerMagnitude (Ability|MakeOutgoingGameplayEffectSpec (Variables|Balance|GetGE_RestoreShields)) "(TagName=\"SetByCaller.Shield\")" (FortPawn|GetMaxShield _owningpawn)))
      (Ability|ApplyGameplayEffectSpecToOwner _returnvalue_2))
    (:else)))
