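; GA_Athena_VoidSlide_HealingOverTime is a passive VoidSlide gameplay ability that periodically restores the owning player's health and shield.
; A re-armed timer checks tuned caps and storm rules, then heals, retries later, or ends the ability when nothing can be restored.
; Each heal draws on a pool that scales with period count up to a limit, spending on health before shield.
; SetByCaller effects apply the amounts, and a gameplay cue fires on success.
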(event Ability|EventFailedToActivatePassiveAbility
  (params (ActorInfo :type Gameplay%20Ability%20Actor%20Info%20Structure))
  ;[1 0 0.7498] Required for Passive GAs
  (|CallParentFunction ActorInfo))

(event Ability|EventActivateAbility
  ;[0 0.5547 0.0738] ActivateAbility
  ; Make sure healing is possible, otherwise set the intial delay and begin healing periods
  (|CallParentFunction)
  (bind _outfortplayerpawn (CallFunction|GetAvatarasValidFortPlayerPawn)
    (:ValidFortPlayerPawn
      (Variables|Default|SetOwningPawn _outfortplayerpawn)
      (CallFunction|DelayNextHealAttempt (ScalableFloat|GetValueatLevel (Variables|Balance|Hotfixes|GetHF_DelayBeforeStartHealing))))
    (:AvatarCastFailed
      (Ability|EndAbility))))

(event Custom|DelayNextHealAttempt
  (params (DelayDuration :type Float%20%28double-precision%29))
  ;[0.2881 0.4466 0.6615] DelayNextHealAttempt
  ; Sets the next heal time based on the previous Input state: initial delay before heals, time after successful heal, or retrys between Failed attempts
  (bind _returnvalue (Utilities|Time|SetTimerbyEvent (EventDispatchers|CreateEvent :!event Custom|DetermineNextHealState) DelayDuration))
  (Variables|InternalVariables|SetTimerHandle_DelayBetweenHealing _returnvalue))

(event Custom|DetermineNextHealState
  ;[0.2881 0.4466 0.6615] Determine Next Heal State
  ; Should heal now, try again later, or end ability
  (bind (_ishealingvalid _shouldendability) (CallFunction|CheckforValidHealing))
  (if _ishealingvalid
    ;;[0.02589 0.2566 1] Healing is Valid - attempt the next heal and then quue the next heal attempt
    (CallFunction|AttemptNextHeal)
    (elif _shouldendability
      ;;[1 0.2521 0] Healing is NOT valid and the conditions won't ever be valid - End the Ability
      (Ability|EndAbility)
      (else
        ;;[0.4219 0.2355 0.1409] Healing is NOT Valid - but should not be cancelled, because conditions may change (ex. player is in the storm)
        (CallFunction|DelayNextHealAttempt (ScalableFloat|GetValueatLevel (Variables|Balance|Hotfixes|GetHF_DelayAfterFailedHealCheck)))))))

(event Custom|AttemptNextHeal
  ;[0 0.1712 1] AttemptNextHeal
  ; Heal the player and then check if healing is still needed to heal again
  (Variables|Default|GetOwningPawn
    (:then
      (Math|Integer|IncrementInt (Variables|InternalVariables|GetCountOfHealPeriods)
        (:%20%20
          (bind _bhealingwasapplied (CallFunction|HealOwningPlayer))
          (if _bhealingwasapplied
            (Ability|ExecuteGameplayCueWithParamsOnOwner (Variables|Default|GetGCNHealing) (Ability|GameplayCue|MakeGameplayCueParameters (Math|Conversions|ToFloat(Integer) (Variables|InternalVariables|GetCountOfHealPeriods)) (Math|Conversions|ToFloat(Integer) (Variables|InternalVariables|GetCountOfHealPeriods)) (Class|AthenaFunctionLibrary|MakeSimpleFortEffectContext (Variables|Default|GetOwningPawn) (Variables|Default|GetOwningPawn)))) ; new
            (else
              _))
          (CallFunction|DelayNextHealAttempt (ScalableFloat|GetValueatLevel (Variables|Balance|Hotfixes|GetHF_IntervalBetweenHeals))))))
    (:else
      (Ability|EndAbility))))

(fn CanRestoreHealth
  (bind _hf_maxhealthrestore (Variables|Balance|Hotfixes|GetHF_MaxHealthRestore))
  (bind _returnvalue (ScalableFloat|GetValueatLevel _hf_maxhealthrestore))
  ;[0.1114 0.3666 0.5781] CanRestoreHealth
  ; Make sure that health can be restored - not already at max value and is enabled by tuning variables
  (bind _owningpawn (Variables|Default|GetOwningPawn)
    (:then
      (if (> _returnvalue 0)
        (return (and (< (FortPawn|GetHealth _owningpawn) (Math|Float|Min(Float) (FortPawn|GetMaxHealth _owningpawn) _returnvalue)) (> (ScalableFloat|GetValueatLevel (Variables|Balance|Hotfixes|GetHF_HealthRestoreFactor)) 0)))
        (else
          _)))
    (:else
      _))
  (return false))

(fn CanRestoreShield
  (bind _hf_maxshieldrestore (Variables|Balance|Hotfixes|GetHF_MaxShieldRestore))
  (bind _returnvalue (ScalableFloat|GetValueatLevel _hf_maxshieldrestore))
  ;[0.1114 0.3666 0.5781] CanRestoreHealth
  ; Make sure that health can be restored - not already at max value and is enabled by tuning variables
  (bind _owningpawn (Variables|Default|GetOwningPawn)
    (:then
      (if (> _returnvalue 0)
        (return (and (< (FortPawn|GetShield _owningpawn) (Math|Float|Min(Float) (FortPawn|GetMaxShield _owningpawn) _returnvalue)) (> (ScalableFloat|GetValueatLevel (Variables|Balance|Hotfixes|GetHF_ShieldRestoreFactor)) 0)))
        (else
          _)))
    (:else
      _))
  (return false))

(fn CheckForValidHealing
  (Utilities|FlowControl|Sequence
    (:then_0
      ;[0.3355 0.5518 0.987] Check if healing in the storm is allowed, then if the player is in the storm determine if healing is allowed/not
      (if (ScalableFloat|GetValueasBool (Variables|Balance|Hotfixes|GetHF_HealingEnabledInStorm))
        (LocalVariables|Seth_isStormHealingValid true)
        (else
          (bind _returnvalue (SafeZone|IsLocationinSafeZone (Transformation|GetActorLocation (Variables|Default|GetOwningPawn))))
          (LocalVariables|Seth_isStormHealingValid _returnvalue))))
    (:then_1
      ;[1 0.4539 0.904] RETURN is healing valid?
      ; The ability should end if there's no way for the healing conditions to become valid after this evaluation (the storm could move, but the player won't take damage inside the hiding prop)
      (return (and (LocalVariables|Geth_isStormHealingValid) (or (CallFunction|CanRestoreHealth) (CallFunction|CanRestoreShield))) (not (or (CallFunction|CanRestoreHealth) (CallFunction|CanRestoreShield)))))))

(fn HealOwningPlayer
  (bind _hf_healthrestorefactor (Variables|Balance|Hotfixes|GetHF_HealthRestoreFactor))
  (bind _returnvalue (ScalableFloat|GetValueatLevel _hf_healthrestorefactor))
  (bind _h_currenthealingpoolmax (LocalVariables|Geth_CurrentHealingPoolMax))
  (LocalVariables|Seth_HasHealingOccured false)
  ;[1 0.8869 0] The healing pool tracks the amout of restoration provided by this heal attempt. Is used for if the Health or Shield amounts that are modified by having a non-standard healing factor (ex. the designer intends for the Siphon to restore shields less effectively than HP)
  (LocalVariables|Seth_CurrentHealingPoolMax (* (ScalableFloat|GetValueatLevel (Variables|Balance|Hotfixes|GetHF_BaseHealingAmount)) (Math|Float|Min(Float) (Math|Float|SelectFloat (* (ScalableFloat|GetValueatLevel (Variables|Balance|Hotfixes|GetHF_HealingMultiplierByPeriod)) (- (Variables|InternalVariables|GetCountOfHealPeriods) (ScalableFloat|GetValueasInteger (Variables|Balance|Hotfixes|GetHF_HealPeriodsBeforeMultiplier)))) 1.0 (> (Variables|InternalVariables|GetCountOfHealPeriods) (ScalableFloat|GetValueasInteger (Variables|Balance|Hotfixes|GetHF_HealPeriodsBeforeMultiplier)))) (ScalableFloat|GetValueatLevel (Variables|Balance|Hotfixes|GetHF_MaxHealingScalar)))))
  (Utilities|FlowControl|Sequence
    (:then_0
      ;[1 1 1 1 48] s[0] ATTEMPT RESTORE HEALTH
      (if (CallFunction|CanRestoreHealth)
        (bind _owningpawn (Variables|Default|GetOwningPawn)
          (:then
            (bind _output_get (LocalVariables|Seth_RestoreHealthAmount (Math|Float|Min(Float) (- (Math|Float|Min(Float) (FortPawn|GetMaxHealth _owningpawn) (ScalableFloat|GetValueatLevel (Variables|Balance|Hotfixes|GetHF_MaxHealthRestore))) (FortPawn|GetHealth _owningpawn)) (* (LocalVariables|Geth_CurrentHealingPoolMax) _returnvalue))))
            (LocalVariables|Seth_CurrentHealingPoolMax (- (LocalVariables|Geth_CurrentHealingPoolMax) (Math|Float|Ceil (Math|Float|SafeDivide _output_get _returnvalue))))
            (bind _returnvalue_2 (Ability|GameplayEffect|AssignTagSetbyCallerMagnitude (Ability|MakeOutgoingGameplayEffectSpec (Variables|Balance|GetGE_RestoreHealth)) "(TagName=\"SetByCaller.Healing\")" (LocalVariables|Geth_RestoreHealthAmount)))
            (Ability|ApplyGameplayEffectSpecToOwner _returnvalue_2)
            (LocalVariables|Seth_HasHealingOccured true))
          (:else))))
    (:then_1
      ;[1 1 1 1 48] s[1] ATTEMPT RESTORE SHIELD
      (if (and (CallFunction|CanRestoreShield) (> _h_currenthealingpoolmax 0))
        (bind _owningpawn_2 (Variables|Default|GetOwningPawn)
          (:then
            (LocalVariables|Seth_RestoreShieldAmount (Math|Float|Min(Float) (- (ScalableFloat|GetValueatLevel (Variables|Balance|Hotfixes|GetHF_MaxShieldRestore)) (FortPawn|GetShield _owningpawn_2)) (Math|Conversions|ToFloat(Integer) (Math|Float|Floor (* _h_currenthealingpoolmax (ScalableFloat|GetValueatLevel (Variables|Balance|Hotfixes|GetHF_ShieldRestoreFactor))))))) ; restore shield amount 
            (bind _returnvalue_3 (Ability|GameplayEffect|AssignTagSetbyCallerMagnitude (Ability|MakeOutgoingGameplayEffectSpec (Variables|Balance|GetGE_RestoreShields)) "(TagName=\"SetByCaller.Shield\")" (LocalVariables|Geth_RestoreShieldAmount)))
            (Ability|ApplyGameplayEffectSpecToOwner _returnvalue_3)
            (LocalVariables|Seth_HasHealingOccured true))
          (:else))))
    (:then_2
      ;[1 0.5202 0.9718] s[2] RETURN
      ; whether healing occured
      ; useful for if FX should be applied
      (return (LocalVariables|Geth_HasHealingOccured)))))
